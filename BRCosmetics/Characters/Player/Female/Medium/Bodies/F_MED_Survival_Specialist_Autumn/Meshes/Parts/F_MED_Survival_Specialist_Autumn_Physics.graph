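# Physics asset for the Survival Specialist Autumn female character body part, setting up dynamic hair simulation.
# Capsule bodies form four hair chains from the head: a forked right A strand, left and right B strands, and a back strand.
# Tight swing and twist limits link each chain, with twist locked on the A strand and one back joint.
# Head, neck and upper spine capsules add body shapes, with only head-to-neck constrained, fully free.
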PhysicsAsset /BRCosmetics/Characters/Player/Female/Medium/Bodies/F_MED_Survival_Specialist_Autumn/Meshes/Parts/F_MED_Survival_Specialist_Autumn_Physics
bodies {
 dny_hair_A_a_r [{shapeType=Capsule, length=3.0675, radius=2.0729, center=<2.952 0.7138 0>, rotation=<rot 90 0 0>}]
 dny_hair_A_b_r [{shapeType=Capsule, length=4.4213, radius=2.0706, center=<3.7153 1.3331 0>, rotation=<rot 90 0 0>}]
 dny_hair_A_c_r [{shapeType=Capsule, length=4.6344, radius=1.9403, center=<4.0732 0 0>, rotation=<rot 90 0 0>}]
 dny_hair_A_d_r [{shapeType=Capsule, length=2.927, radius=1.9919, center=<0.7424 0.0001 0.7631>, rotation=<rot -75.957 0.0001 0>}]
 dny_hair_A_tip_a_r [{shapeType=Capsule, length=9, radius=2.2922, center=<0 0 0.073>, rotation=<rot 90 0 0>}]
 dny_hair_A_tip_b_r [{shapeType=Capsule, length=2.2133, radius=1.2341, rotation=<rot -75.9418 -0.0002 0.0002>}]
 dny_hair_B_a_base_l [{shapeType=Capsule, length=2.1972, radius=1.1066, rotation=<rot 0 -90 0>}]
 dny_hair_B_a_base_r [{shapeType=Capsule, length=2.1972, radius=1.1066, rotation=<rot 0 -90 0>}]
 dny_hair_B_a_l [{shapeType=Capsule, length=3.5, radius=3, center=<0.6192 0.0001 0>, rotation=<rot 90 0 0>}]
 dny_hair_B_a_r [{shapeType=Capsule, length=2.75, radius=1.75, center=<0.6192 0.0001 0>, rotation=<rot 90 0 0>}]
 dny_hair_B_b_l [{shapeType=Capsule, length=3.5, radius=3, center=<0.6192 0.0001 0>, rotation=<rot 90 0 0>}]
 dny_hair_B_b_r [{shapeType=Capsule, length=3.75, radius=2, center=<0.0261 0.0002 -0.5931>, rotation=<rot 44.9987 179.9999 -180>}]
 dny_hair_B_c_l [{shapeType=Capsule, length=3.5, radius=3, center=<0.6192 0.0001 0>, rotation=<rot 90 0 0>}]
 dny_hair_B_c_r [{shapeType=Capsule, length=6, radius=2.25, center=<2.4582 0.1738 -0.0001>, rotation=<rot 50.0002 159.9961 180>}]
 dny_hair_B_d_l [{shapeType=Capsule, length=3.5, radius=3, center=<0.6192 0.0001 0>, rotation=<rot 90 0 0>}]
 dny_hair_B_d_r [{shapeType=Capsule, length=5.75, radius=2, center=<1.4511 -0.4864 -0.2652>, rotation=<rot 44.9967 169.9982 -180>}]
 dny_hair_B_e_l [{shapeType=Capsule, length=3.5, radius=3, center=<0.6431 0.6526 0.3795>, rotation=<rot 90 90 97.0048>}]
 dny_hair_B_e_r [{shapeType=Capsule, length=2.25, radius=2.5, center=<0.6192 -0.6488 0>, rotation=<rot 29.9989 180 180>}]
 dny_hair_B_f_l [{shapeType=Capsule, length=5, radius=3, center=<1.2289 0.5067 0>, rotation=<rot 90 0 0>}]
 dny_hair_B_f_r [{shapeType=Capsule, length=2.25, radius=3, center=<0.6192 -0.6411 0>, rotation=<rot 49.998 180 180>}]
 dny_hair_bk_a [{shapeType=Capsule, length=2.25, radius=2.25, center=<0.6192 0.0001 0>, rotation=<rot 90 0 0>}]
 dny_hair_bk_a_base [{shapeType=Capsule, length=2.1972, radius=1.1066, rotation=<rot 0 -90 0>}]
 dny_hair_bk_b [{shapeType=Capsule, length=2.25, radius=2.25, center=<0.6192 0.0001 0>, rotation=<rot 90 0 0>}]
 dny_hair_bk_c [{shapeType=Capsule, length=3.25, radius=2.75, center=<0.6192 1.1128 -0.0001>, rotation=<rot 90 -14.0363 5.9637>}]
 dny_hair_bk_d [{shapeType=Capsule, length=3.25, radius=2.75, center=<0.6192 0.0001 0>, rotation=<rot 90 0 0>}]
 dny_hair_bk_e [{shapeType=Capsule, length=3.25, radius=2.75, center=<0.6192 -0.4126 0>, rotation=<rot 90 0 0>}]
 dny_hair_bk_f [{shapeType=Capsule, length=3.5, radius=2.5, center=<0.6192 -0.9296 0>, rotation=<rot 90 0 0>}]
 head [{
   shapeType=Capsule
   length=13.6345
   radius=5.6488
   shapeName=head_capsule
   center=<0.866 1.3706 0.2817>
   rotation=<rot 0.6162 90.8112 -90.0539>
  }]
 neck_01 [{shapeType=Capsule, length=29.9237, radius=4.8411, center=<2.7629 0.5103 -0.2038>, rotation=<rot -0.0006 175.8263 180>}]
 spine_04 [{shapeType=Capsule, length=21.9953, radius=11.1, center=<3.0166 -0.3904 -0.0002>}]
 spine_05 [{shapeType=Capsule, length=28.0686, radius=7.5423, center=<4.4169 -2.8164 -0.0005>}]
}
constraints [
 {bone1Name=dny_hair_B_a_base_l, bone2Name=head, swing1LimitDegrees=8, swing2LimitDegrees=8, twistLimitDegrees=10}
 {bone1Name=dny_hair_B_a_l, bone2Name=dny_hair_B_a_base_l, swing1LimitDegrees=10, swing2LimitDegrees=10, twistLimitDegrees=10}
 {bone1Name=dny_hair_B_b_l, bone2Name=dny_hair_B_a_l, swing1LimitDegrees=8, swing2LimitDegrees=8, twistLimitDegrees=10}
 {bone1Name=dny_hair_B_c_l, bone2Name=dny_hair_B_b_l, swing1LimitDegrees=8, swing2LimitDegrees=8, twistLimitDegrees=10}
 {bone1Name=dny_hair_B_d_l, bone2Name=dny_hair_B_c_l, swing1LimitDegrees=8, swing2LimitDegrees=8, twistLimitDegrees=10}
 {bone1Name=dny_hair_B_e_l, bone2Name=dny_hair_B_d_l, swing1LimitDegrees=8, swing2LimitDegrees=8, twistLimitDegrees=10}
 {bone1Name=dny_hair_B_f_l, bone2Name=dny_hair_B_e_l, swing1LimitDegrees=8, swing2LimitDegrees=8, twistLimitDegrees=10}
 {bone1Name=dny_hair_B_a_base_r, bone2Name=head, swing1LimitDegrees=15, swing2LimitDegrees=15, twistLimitDegrees=10}
 {bone1Name=dny_hair_B_a_r, bone2Name=dny_hair_B_a_base_r, swing1LimitDegrees=15, swing2LimitDegrees=15, twistLimitDegrees=10}
 {bone1Name=dny_hair_B_b_r, bone2Name=dny_hair_B_a_r, swing1LimitDegrees=15, swing2LimitDegrees=15, twistLimitDegrees=10}
 {bone1Name=dny_hair_B_c_r, bone2Name=dny_hair_B_b_r, swing1LimitDegrees=15, swing2LimitDegrees=15, twistLimitDegrees=10}
 {bone1Name=dny_hair_B_d_r, bone2Name=dny_hair_B_c_r, swing1LimitDegrees=15, swing2LimitDegrees=15, twistLimitDegrees=10}
 {bone1Name=dny_hair_B_e_r, bone2Name=dny_hair_B_d_r, swing1LimitDegrees=15, swing2LimitDegrees=15, twistLimitDegrees=10}
 {bone1Name=dny_hair_B_f_r, bone2Name=dny_hair_B_e_r, swing1LimitDegrees=15, swing2LimitDegrees=15, twistLimitDegrees=10}
 {bone1Name=dny_hair_bk_f, bone2Name=dny_hair_bk_e, swing1LimitDegrees=15, swing2LimitDegrees=15, twistLimitDegrees=15}
 {bone1Name=dny_hair_bk_e, bone2Name=dny_hair_bk_d, swing1LimitDegrees=15, swing2LimitDegrees=15, twistLimitDegrees=15}
 {bone1Name=dny_hair_bk_c, bone2Name=dny_hair_bk_b, swing1LimitDegrees=15, swing2LimitDegrees=15, twistLimitDegrees=1, twistMotion=Locked}
 {bone1Name=dny_hair_bk_d, bone2Name=dny_hair_bk_c, swing1LimitDegrees=15, swing2LimitDegrees=15, twistLimitDegrees=15}
 {bone1Name=dny_hair_bk_b, bone2Name=dny_hair_bk_a, swing1LimitDegrees=15, swing2LimitDegrees=15, twistLimitDegrees=15}
 {bone1Name=dny_hair_bk_a, bone2Name=dny_hair_bk_a_base, swing1LimitDegrees=15, swing2LimitDegrees=15, twistLimitDegrees=15}
 {bone1Name=dny_hair_bk_a_base, bone2Name=head, swing1LimitDegrees=0.1, swing2LimitDegrees=0.1, twistLimitDegrees=1}
 {bone1Name=dny_hair_A_a_r, bone2Name=head, swing1LimitDegrees=15, swing2LimitDegrees=15, twistLimitDegrees=0.5, twistMotion=Locked}
 {bone1Name=dny_hair_A_b_r, bone2Name=dny_hair_A_a_r, swing1LimitDegrees=15, swing2LimitDegrees=15, twistLimitDegrees=0.5, twistMotion=Locked}
 {bone1Name=dny_hair_A_c_r, bone2Name=dny_hair_A_b_r, swing1LimitDegrees=15, swing2LimitDegrees=15, twistLimitDegrees=0.5, twistMotion=Locked}
 {bone1Name=dny_hair_A_d_r, bone2Name=dny_hair_A_c_r, swing1LimitDegrees=15, swing2LimitDegrees=15, twistLimitDegrees=0.5, twistMotion=Locked}
 {
  bone1Name=dny_hair_A_tip_b_r
  bone2Name=dny_hair_A_d_r
  swing1LimitDegrees=15
  swing2LimitDegrees=15
  twistLimitDegrees=0.5
  twistMotion=Locked
 }
 {
  bone1Name=dny_hair_A_tip_a_r
  bone2Name=dny_hair_A_d_r
  swing1LimitDegrees=15
  swing2LimitDegrees=15
  twistLimitDegrees=0.5
  twistMotion=Locked
 }
 {
  bone1Name=head
  bone2Name=neck_01
  swing1LimitDegrees=45
  swing1Motion=Free
  swing2LimitDegrees=45
  swing2Motion=Free
  twistLimitDegrees=45
  twistMotion=Free
 }
]
