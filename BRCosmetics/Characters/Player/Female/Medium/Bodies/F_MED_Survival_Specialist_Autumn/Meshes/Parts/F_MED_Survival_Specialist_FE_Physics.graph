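# Physics asset for the female Survival Specialist (Autumn) FE mesh part, setting up dynamic hair that can collide with the upper body.
# Capsules cover the hair bones plus head, neck, upper spine, clavicles and upper arms.
# Constraints form left, right and back strands plus a right-side lock with a forked tip, all rooted at the head.
# Strands have tight twist limits, the back root is nearly rigid, and head-to-neck is free.
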PhysicsAsset /BRCosmetics/Characters/Player/Female/Medium/Bodies/F_MED_Survival_Specialist_Autumn/Meshes/Parts/F_MED_Survival_Specialist_FE_Physics
bodies {
 clavicle_l [{shapeType=Capsule, length=12, radius=5.5, shapeName=clavicle_l_capsule, center=<3.5 -1 -1>, rotation=<rot 80 -0.0003 -0.0003>}]
 clavicle_r [{shapeType=Capsule, length=12, radius=5.5, shapeName=clavicle_r_capsule, center=<-3.5 1 1>, rotation=<rot 80 -0.0003 -0.0004>}]
 dny_hair_A_a_r [{shapeType=Capsule, length=3.0675, radius=1.5, center=<2.952 0.7138 0>, rotation=<rot 80 -179.9999 179.9999>}]
 dny_hair_A_b_r [{shapeType=Capsule, length=4, radius=2, center=<3.7153 0 0.5>, rotation=<rot 90 0 0>}]
 dny_hair_A_c_r [{shapeType=Capsule, length=4.6344, radius=2.5, center=<4.1372 -0.3 0>, rotation=<rot 90 1.7899 -8.2103>}]
 dny_hair_A_d_r [{shapeType=Capsule, length=2.927, radius=1.9919, center=<0.7424 0.2 0.7>, rotation=<rot -75.957 0.0001 0>}]
 dny_hair_A_tip_a_r [{shapeType=Capsule, length=4, radius=2.2922, center=<2.502 0 0>, rotation=<rot 70 0.0002 0.0001>}]
 dny_hair_A_tip_b_r [{shapeType=Capsule, length=2.2133, radius=1.2341, rotation=<rot -75.9418 -0.0002 0.0002>}]
 dny_hair_B_a_base_l [{shapeType=Capsule, length=2.1972, radius=1.1066, rotation=<rot 0 -90 0>}]
 dny_hair_B_a_base_r [{shapeType=Capsule, length=2.1972, radius=1.1066, rotation=<rot 0 -90 0>}]
 dny_hair_B_a_l [{shapeType=Capsule, length=6, radius=2.5, center=<4 0 0>, rotation=<rot 90 0 0>}]
 dny_hair_B_a_r [{shapeType=Capsule, length=6, radius=2.5, center=<4 0 0>, rotation=<rot 90 0 0>}]
 dny_hair_B_b_l [{shapeType=Capsule, length=6, radius=2.5, center=<4 0 0>, rotation=<rot 90 0 0>}]
 dny_hair_B_b_r [{shapeType=Capsule, length=6, radius=2.5, center=<4 0 0>, rotation=<rot 90 0 0>}]
 dny_hair_B_c_l [{shapeType=Capsule, length=6, radius=2.5, center=<4 0 1>, rotation=<rot 90 0 0>}]
 dny_hair_B_c_r [{shapeType=Capsule, length=6, radius=2.5, center=<4 0 0>, rotation=<rot 90 0 0>}]
 dny_hair_B_d_l [{shapeType=Capsule, length=6, radius=2.5, center=<4 1 1>, rotation=<rot 90 0 0>}]
 dny_hair_B_d_r [{shapeType=Capsule, length=6, radius=2.5, center=<4 1 -1>, rotation=<rot 90 0 0>}]
 dny_hair_B_e_l [{shapeType=Capsule, length=6, radius=2.5, center=<4 0 0>, rotation=<rot 90 0 0>}]
 dny_hair_B_e_r [{shapeType=Capsule, length=6, radius=2.5, center=<4 0 0>, rotation=<rot 90 0 0>}]
 dny_hair_B_f_l [{shapeType=Capsule, length=6, radius=4, center=<4 0 0>, rotation=<rot 90 0 0>}]
 dny_hair_B_f_r [{shapeType=Capsule, length=6, radius=2.5, center=<4 0 0>, rotation=<rot 90 0 0>}]
 dny_hair_bk_a [{shapeType=Capsule, length=6, radius=2.5, center=<4 0 0>, rotation=<rot 90 0 0>}]
 dny_hair_bk_a_base [{shapeType=Capsule, length=2.1972, radius=1.1066, rotation=<rot 0 -90 0>}]
 dny_hair_bk_b [{shapeType=Capsule, length=6, radius=2.5, center=<4 0 0>, rotation=<rot 90 0 0>}]
 dny_hair_bk_c [{shapeType=Capsule, length=6, radius=2.5, center=<4 0 0>, rotation=<rot 90 0 0>}]
 dny_hair_bk_d [{shapeType=Capsule, length=6, radius=2.5, center=<4 0 0>, rotation=<rot 90 0 0>}]
 dny_hair_bk_e [{shapeType=Capsule, length=6, radius=2.5, center=<4 0 0>, rotation=<rot 90 0 0>}]
 dny_hair_bk_f [{shapeType=Capsule, length=6, radius=2.5, center=<4 0 0>, rotation=<rot 90 0 0>}]
 head [{shapeType=Capsule, length=6, radius=6, shapeName=head_capsule, center=<3.5 3 0>, rotation=<rot 90 0 0>}]
 neck_01 [{shapeType=Capsule, length=8, radius=4.5, shapeName=neck_01_capsule, center=<3.6683 -0.2876 0>, rotation=<rot 90 0 0>}]
 spine_03 [{shapeType=Capsule, length=15, radius=10, shapeName=spine_04_capsule, center=<0.229 2 0>, rotation=<rot 90 0 10.0001>}]
 spine_05 [
  {shapeType=Capsule, length=12, radius=9.25, shapeName=spine_05_capsule, center=<-2.5963 0.209 0>}
  {shapeType=Capsule, length=12, radius=7.75, shapeName=spine_05_capsule, center=<-6.4845 4.2091 0>}
  {shapeType=Capsule, length=3, radius=2.5, shapeName=spine_05_capsule, center=<0.4654 10.0373 0>, rotation=<rot 90 -172.875 -139.439>}
 ]
 upperarm_l [{shapeType=Capsule, length=20, radius=3.5, shapeName=upperarm_l_capsule, center=<12 -0.5 0>, rotation=<rot 90 0 0>}]
 upperarm_r [{shapeType=Capsule, length=20, radius=3.5, shapeName=upperarm_r_capsule, center=<-12 0.5 0>, rotation=<rot 90 0 0>}]
}
constraints [
 {bone1Name=dny_hair_B_a_base_l, bone2Name=head, swing1LimitDegrees=8, swing2LimitDegrees=8, twistLimitDegrees=10}
 {bone1Name=dny_hair_B_a_l, bone2Name=dny_hair_B_a_base_l, swing1LimitDegrees=30, swing2LimitDegrees=30, twistLimitDegrees=1}
 {bone1Name=dny_hair_B_b_l, bone2Name=dny_hair_B_a_l, swing1LimitDegrees=30, swing2LimitDegrees=30, twistLimitDegrees=1}
 {bone1Name=dny_hair_B_c_l, bone2Name=dny_hair_B_b_l, swing1LimitDegrees=30, swing2LimitDegrees=30, twistLimitDegrees=1}
 {bone1Name=dny_hair_B_d_l, bone2Name=dny_hair_B_c_l, swing1LimitDegrees=30, swing2LimitDegrees=30, twistLimitDegrees=1}
 {bone1Name=dny_hair_B_e_l, bone2Name=dny_hair_B_d_l, swing1LimitDegrees=30, swing2LimitDegrees=30, twistLimitDegrees=1}
 {bone1Name=dny_hair_B_f_l, bone2Name=dny_hair_B_e_l, swing1LimitDegrees=30, swing2LimitDegrees=30, twistLimitDegrees=1}
 {bone1Name=dny_hair_B_a_base_r, bone2Name=head, swing1LimitDegrees=15, swing2LimitDegrees=15, twistLimitDegrees=10}
 {bone1Name=dny_hair_B_a_r, bone2Name=dny_hair_B_a_base_r, swing1LimitDegrees=30, swing2LimitDegrees=30, twistLimitDegrees=1}
 {bone1Name=dny_hair_B_b_r, bone2Name=dny_hair_B_a_r, swing1LimitDegrees=30, swing2LimitDegrees=30, twistLimitDegrees=1}
 {bone1Name=dny_hair_B_c_r, bone2Name=dny_hair_B_b_r, swing1LimitDegrees=30, swing2LimitDegrees=30, twistLimitDegrees=1}
 {bone1Name=dny_hair_B_d_r, bone2Name=dny_hair_B_c_r, swing1LimitDegrees=30, swing2LimitDegrees=30, twistLimitDegrees=1}
 {bone1Name=dny_hair_B_e_r, bone2Name=dny_hair_B_d_r, swing1LimitDegrees=30, swing2LimitDegrees=30, twistLimitDegrees=1}
 {bone1Name=dny_hair_B_f_r, bone2Name=dny_hair_B_e_r, swing1LimitDegrees=30, swing2LimitDegrees=30, twistLimitDegrees=1}
 {bone1Name=dny_hair_bk_f, bone2Name=dny_hair_bk_e, swing1LimitDegrees=30, swing2LimitDegrees=30, twistLimitDegrees=1}
 {bone1Name=dny_hair_bk_e, bone2Name=dny_hair_bk_d, swing1LimitDegrees=30, swing2LimitDegrees=30, twistLimitDegrees=1}
 {bone1Name=dny_hair_bk_c, bone2Name=dny_hair_bk_b, swing1LimitDegrees=30, swing2LimitDegrees=30, twistLimitDegrees=1}
 {bone1Name=dny_hair_bk_d, bone2Name=dny_hair_bk_c, swing1LimitDegrees=30, swing2LimitDegrees=30, twistLimitDegrees=1}
 {bone1Name=dny_hair_bk_b, bone2Name=dny_hair_bk_a, swing1LimitDegrees=30, swing2LimitDegrees=30, twistLimitDegrees=1}
 {bone1Name=dny_hair_bk_a, bone2Name=dny_hair_bk_a_base, swing1LimitDegrees=30, swing2LimitDegrees=30, twistLimitDegrees=1}
 {bone1Name=dny_hair_bk_a_base, bone2Name=head, swing1LimitDegrees=0.1, swing2LimitDegrees=0.1, twistLimitDegrees=1}
 {bone1Name=dny_hair_A_a_r, bone2Name=head, swing1LimitDegrees=15, swing2LimitDegrees=15, twistLimitDegrees=0.5, twistMotion=Locked}
 {bone1Name=dny_hair_A_b_r, bone2Name=dny_hair_A_a_r, swing1LimitDegrees=15, swing2LimitDegrees=15, twistLimitDegrees=0.5}
 {bone1Name=dny_hair_A_c_r, bone2Name=dny_hair_A_b_r, swing1LimitDegrees=15, swing2LimitDegrees=15, twistLimitDegrees=0.5}
 {bone1Name=dny_hair_A_d_r, bone2Name=dny_hair_A_c_r, swing1LimitDegrees=15, swing2LimitDegrees=15, twistLimitDegrees=0.5}
 {
  bone1Name=dny_hair_A_tip_b_r
  bone2Name=dny_hair_A_d_r
  swing1LimitDegrees=15
  swing2LimitDegrees=15
  twistLimitDegrees=0.5
  twistMotion=Locked
 }
 {
  bone1Name=dny_hair_A_tip_a_r
  bone2Name=dny_hair_A_d_r
  swing1LimitDegrees=15
  swing2LimitDegrees=15
  twistLimitDegrees=0.5
  twistMotion=Locked
 }
 {
  bone1Name=head
  bone2Name=neck_01
  swing1LimitDegrees=45
  swing1Motion=Free
  swing2LimitDegrees=45
  swing2Motion=Free
  twistLimitDegrees=45
  twistMotion=Free
 }
]
