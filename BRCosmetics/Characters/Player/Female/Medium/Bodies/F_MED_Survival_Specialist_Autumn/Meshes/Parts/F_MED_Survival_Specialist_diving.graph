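# Physics asset for the diving mesh part of the female Autumn Survival Specialist, setting up capsule bodies and constraints for dynamic hair.
# Constraints chain a right-side A strand with a forked tip, plus left, right and back strands, up to the head.
# Cross-links tie the side strands to the back strand at base, middle and end, and to each other near the root.
# Head, neck and upper-spine capsules give the hair body shapes to collide with.
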PhysicsAsset /BRCosmetics/Characters/Player/Female/Medium/Bodies/F_MED_Survival_Specialist_Autumn/Meshes/Parts/F_MED_Survival_Specialist_diving
bodies {
 dny_hair_A_a_r [{shapeType=Capsule, length=3.0675, radius=2.0729, rotation=<rot 88.7711 -0.0009 -0.0008>}]
 dny_hair_A_b_r [{shapeType=Capsule, length=6.1713, radius=2.8206, center=<2.1501 0 -0.0459>, rotation=<rot 88.7763 -0.001 -0.001>}]
 dny_hair_A_c_r [{shapeType=Capsule, length=6.6344, radius=2.4403, center=<1.5946 -0.0001 -0.0281>, rotation=<rot -88.9897 179.9999 -179.9999>}]
 dny_hair_A_d_r [{shapeType=Capsule, length=2.927, radius=1.9919, center=<0.8783 0 0.2197>, rotation=<rot -75.957 0.0001 0>}]
 dny_hair_A_tip_a_r [{shapeType=Capsule, length=6.7853, radius=2.2922, center=<1.081 0 0.1614>, rotation=<rot -80.4696 -179.9998 179.9998>}]
 dny_hair_A_tip_b_r [{shapeType=Capsule, length=2.2133, radius=1.2341, rotation=<rot -75.9418 -0.0002 0.0002>}]
 dny_hair_B_a_base_l [{shapeType=Capsule, length=2.1972, radius=1.1066, rotation=<rot 0 -90 0>}]
 dny_hair_B_a_base_r [{shapeType=Capsule, length=2.1972, radius=1.1066, rotation=<rot 0 -90 0>}]
 dny_hair_B_a_l [{shapeType=Capsule, length=2.5, radius=2, center=<1.579 0 0>, rotation=<rot 90 0 0>}]
 dny_hair_B_a_r [{shapeType=Capsule, length=2.5, radius=2, center=<1.579 0 0>, rotation=<rot 90 0 0>}]
 dny_hair_B_b_l [{shapeType=Capsule, length=2.5, radius=2, center=<1.579 0 0>, rotation=<rot 90 0 0>}]
 dny_hair_B_b_r [{shapeType=Capsule, length=3, radius=1.5, center=<1.579 0 0>, rotation=<rot 90 0 0>}]
 dny_hair_B_c_l [{shapeType=Capsule, length=2.5, radius=2, center=<1.579 0 0>, rotation=<rot 90 0 0>}]
 dny_hair_B_c_r [{shapeType=Capsule, length=2.25, radius=1.5, center=<1.579 -0.1255 0>, rotation=<rot 90 0 0>}]
 dny_hair_B_d_l [{shapeType=Capsule, length=2.5, radius=2, center=<1.579 0 0>, rotation=<rot 90 0 0>}]
 dny_hair_B_d_r [{shapeType=Capsule, length=2.75, radius=1.75, center=<1.579 0 0>, rotation=<rot 90 0 0>}]
 dny_hair_B_e_l [{shapeType=Capsule, length=2.5, radius=2, center=<1.579 0 0>, rotation=<rot 90 0 0>}]
 dny_hair_B_e_r [{shapeType=Capsule, length=2, radius=1.75, center=<1.579 0 0>, rotation=<rot 90 0 0>}]
 dny_hair_B_f_l [{shapeType=Capsule, length=2.5, radius=2, center=<1.579 0 0>, rotation=<rot 90 0 0>}]
 dny_hair_B_f_r [{shapeType=Capsule, length=2.5, radius=1.5, center=<1.579 0 0>, rotation=<rot 90 0 0>}]
 dny_hair_bk_a [{shapeType=Capsule, length=2.5, radius=2, center=<1.579 0 0>, rotation=<rot 90 0 0>}]
 dny_hair_bk_a_base [{shapeType=Capsule, length=2.1972, radius=1.1066, rotation=<rot 0 -90 0>}]
 dny_hair_bk_b [{shapeType=Capsule, length=2.5, radius=2, center=<1.579 0 0>, rotation=<rot 90 0 0>}]
 dny_hair_bk_c [{shapeType=Capsule, length=2.5, radius=2, center=<1.579 0 0>, rotation=<rot 90 0 0>}]
 dny_hair_bk_d [{shapeType=Capsule, length=2.5, radius=2, center=<1.579 0 0>, rotation=<rot 90 0 0>}]
 dny_hair_bk_e [{shapeType=Capsule, length=2.5, radius=2, center=<1.579 0 0>, rotation=<rot 90 0 0>}]
 dny_hair_bk_f [{shapeType=Capsule, length=2.5, radius=2, center=<1.579 0 0>, rotation=<rot 90 0 0>}]
 head [{shapeType=Capsule, length=3.279, radius=5.6478, center=<0.3321 1.8573 -0.4521>, rotation=<rot 90 0 -8.865>}]
 neck_01 [{shapeType=Capsule, length=2.4237, radius=4.8411, center=<-0.0564 0.8627 -0.2039>, rotation=<rot 90 0 4.1737>}]
 spine_04 [{shapeType=Capsule, length=22.7453, radius=10.0704, center=<5.1383 -0.3266 -0.0001>}]
 spine_05 [{shapeType=Capsule, length=24.0686, radius=6.7923, center=<5.4435 -0.723 -0.0004>}]
}
constraints [
 {bone1Name=dny_hair_B_a_base_l, bone2Name=head, swing1LimitDegrees=0.1, swing2LimitDegrees=0.1, twistLimitDegrees=1, twistMotion=Locked}
 {
  bone1Name=dny_hair_B_a_l
  bone2Name=dny_hair_B_a_base_l
  swing1LimitDegrees=15
  swing2LimitDegrees=15
  twistLimitDegrees=15
  twistMotion=Locked
 }
 {bone1Name=dny_hair_B_b_l, bone2Name=dny_hair_B_a_l, swing1LimitDegrees=10, swing2LimitDegrees=10, twistLimitDegrees=10, twistMotion=Locked}
 {bone1Name=dny_hair_B_c_l, bone2Name=dny_hair_B_b_l, swing1LimitDegrees=10, swing2LimitDegrees=10, twistLimitDegrees=1}
 {bone1Name=dny_hair_B_d_l, bone2Name=dny_hair_B_c_l, swing1LimitDegrees=10, swing2LimitDegrees=10, twistLimitDegrees=1}
 {bone1Name=dny_hair_B_e_l, bone2Name=dny_hair_B_d_l, swing1LimitDegrees=10, swing2LimitDegrees=10, twistLimitDegrees=1}
 {bone1Name=dny_hair_B_f_l, bone2Name=dny_hair_B_e_l, swing1LimitDegrees=10, swing2LimitDegrees=10, twistLimitDegrees=1}
 {bone1Name=dny_hair_B_a_base_r, bone2Name=head, swing1LimitDegrees=0.1, swing2LimitDegrees=0.1, twistLimitDegrees=1, twistMotion=Locked}
 {
  bone1Name=dny_hair_B_a_r
  bone2Name=dny_hair_B_a_base_r
  swing1LimitDegrees=15
  swing2LimitDegrees=15
  twistLimitDegrees=15
  twistMotion=Locked
 }
 {bone1Name=dny_hair_B_b_r, bone2Name=dny_hair_B_a_r, swing1LimitDegrees=10, swing2LimitDegrees=10, twistLimitDegrees=10, twistMotion=Locked}
 {bone1Name=dny_hair_B_c_r, bone2Name=dny_hair_B_b_r, swing1LimitDegrees=10, swing2LimitDegrees=10, twistLimitDegrees=1}
 {bone1Name=dny_hair_B_d_r, bone2Name=dny_hair_B_c_r, swing1LimitDegrees=10, swing2LimitDegrees=10, twistLimitDegrees=1}
 {bone1Name=dny_hair_B_e_r, bone2Name=dny_hair_B_d_r, swing1LimitDegrees=10, swing2LimitDegrees=10, twistLimitDegrees=1}
 {bone1Name=dny_hair_B_f_r, bone2Name=dny_hair_B_e_r, swing1LimitDegrees=10, swing2LimitDegrees=10, twistLimitDegrees=1}
 {bone1Name=dny_hair_bk_f, bone2Name=dny_hair_bk_e, swing1LimitDegrees=15, swing2LimitDegrees=15, twistLimitDegrees=1}
 {bone1Name=dny_hair_bk_e, bone2Name=dny_hair_bk_d, swing1LimitDegrees=15, swing2LimitDegrees=15, twistLimitDegrees=1}
 {bone1Name=dny_hair_bk_c, bone2Name=dny_hair_bk_b, swing1LimitDegrees=15, swing2LimitDegrees=15, twistLimitDegrees=1}
 {bone1Name=dny_hair_bk_d, bone2Name=dny_hair_bk_c, swing1LimitDegrees=15, swing2LimitDegrees=15, twistLimitDegrees=1}
 {bone1Name=dny_hair_bk_b, bone2Name=dny_hair_bk_a, swing1LimitDegrees=15, swing2LimitDegrees=15, twistLimitDegrees=15}
 {bone1Name=dny_hair_bk_a, bone2Name=dny_hair_bk_a_base, swing1LimitDegrees=15, swing2LimitDegrees=15, twistLimitDegrees=15}
 {bone1Name=dny_hair_bk_a_base, bone2Name=head, swing1LimitDegrees=0.1, swing2LimitDegrees=0.1, twistLimitDegrees=1}
 {bone1Name=dny_hair_B_a_r, bone2Name=dny_hair_B_a_l, swing1LimitDegrees=10, swing2LimitDegrees=10, twistLimitDegrees=10, twistMotion=Locked}
 {bone1Name=dny_hair_A_a_r, bone2Name=head, swing1LimitDegrees=15, swing2LimitDegrees=15, twistLimitDegrees=0.5, twistMotion=Locked}
 {bone1Name=dny_hair_A_b_r, bone2Name=dny_hair_A_a_r, swing1LimitDegrees=15, swing2LimitDegrees=15, twistLimitDegrees=0.5, twistMotion=Locked}
 {bone1Name=dny_hair_A_c_r, bone2Name=dny_hair_A_b_r, swing1LimitDegrees=15, swing2LimitDegrees=15, twistLimitDegrees=0.5, twistMotion=Locked}
 {bone1Name=dny_hair_A_d_r, bone2Name=dny_hair_A_c_r, swing1LimitDegrees=15, swing2LimitDegrees=15, twistLimitDegrees=0.5, twistMotion=Locked}
 {
  bone1Name=dny_hair_A_tip_b_r
  bone2Name=dny_hair_A_d_r
  swing1LimitDegrees=15
  swing2LimitDegrees=15
  twistLimitDegrees=0.5
  twistMotion=Locked
 }
 {
  bone1Name=dny_hair_A_tip_a_r
  bone2Name=dny_hair_A_d_r
  swing1LimitDegrees=15
  swing2LimitDegrees=15
  twistLimitDegrees=0.5
  twistMotion=Locked
 }
 {bone1Name=dny_hair_B_f_r, bone2Name=dny_hair_bk_f, swing1LimitDegrees=10, swing2LimitDegrees=10, twistLimitDegrees=10}
 {bone1Name=dny_hair_B_f_l, bone2Name=dny_hair_bk_f, swing1LimitDegrees=10, swing2LimitDegrees=10, twistLimitDegrees=10}
 {
  bone1Name=dny_hair_B_a_base_r
  bone2Name=dny_hair_bk_a_base
  swing1LimitDegrees=0.1
  swing2LimitDegrees=0.1
  twistLimitDegrees=5
  twistMotion=Locked
 }
 {
  bone1Name=dny_hair_B_a_base_l
  bone2Name=dny_hair_bk_a_base
  swing1LimitDegrees=0.1
  swing2LimitDegrees=0.1
  twistLimitDegrees=5
  twistMotion=Locked
 }
 {bone1Name=dny_hair_B_c_r, bone2Name=dny_hair_bk_c, swing1LimitDegrees=10, swing2LimitDegrees=10, twistLimitDegrees=10}
 {bone1Name=dny_hair_B_c_l, bone2Name=dny_hair_bk_c, swing1LimitDegrees=10, swing2LimitDegrees=10, twistLimitDegrees=10}
]
