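# Animation Blueprint for the Autumn Survival Specialist female body part, driving the character's dynamic hair.
# It copies the attached parent's pose, then caches it after additive hair-bone corrections driven by head-to-spine rotation.
# In the front end, it uses RigidBody simulation or, with physics disabled, bone constraints.
# In game, it blends default, skydiving and swim-sprint RigidBody setups, then applies hair offsets, some scaled by neck bend and targeting.
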AnimBlueprint /BRCosmetics/Characters/Player/Female/Medium/Bodies/F_MED_Survival_Specialist_Autumn/Meshes/Parts/F_MED_Survival_Specialist_Autumn_AnimBP
target_skeleton=&/F_MED_Survival_Specialist_Autumn_Skeleton
graphs [
 AnimGraph {nodes [
   "Transform (Modify) Bone - Bone: dny_hair_B_b_r"(Alpha=1, Translation="0,0,0", Rotation="0.000,-20,0.000", Scale="1,1,1") {
    rotationMode=BMM_Additive
    rotationSpace=BCS_BoneSpace
    scaleMode=BMM_Additive
    scaleSpace=BCS_BoneSpace
    translationMode=BMM_Additive
    translationSpace=BCS_BoneSpace
    alphaBoolBlend {blendOption=Linear}
    alphaScaleBiasClamp {bMapRange=true, inRange {max=15}}
    boneToModify {boneName=dny_hair_B_b_r}
    rotation {yaw=-20}
   }
   n1 = RigidBody(ComponentPose=Local To Component(LocalPose=Use cached pose 'AnimCache'())) {
    bClampLinearTranslationLimitToRefPose=true
    bOverrideWorldGravity=true
    overridePhysicsAsset=&/F_MED_Survival_Specialist_FE_Physics
    simulationSpace=BaseBoneSpace
    alphaBoolBlend {blendOption=Linear}
    baseBoneRef {boneName=spine_05}
    componentAppliedLinearAccClamp=<1000 1000 1000>
    componentLinearAccScale=<0.36 0.36 0.36>
    componentLinearVelScale=<0.3 0.3 0.3>
    overrideWorldGravity {z=-980}
   }
   n2 = Component To Local(ComponentPose=@n1)
   n3 = "Constraint - dny_hair_A_b_r (PP)"!(ComponentPose=Local To Component(LocalPose=Use cached pose 'AnimCache'()), Alpha=1, ConstraintWeights_0=0.5, ConstraintWeights_1:float=0.5) -> (Pose) {
    alphaBoolBlend {blendOption=Linear}
    boneToModify {boneName=dny_hair_A_b_r}
    constraintSetup [{targetBone {boneName=neck_01}}, {targetBone {boneName=head}}]
    constraintWeights [1,1]
   }
   n4 = "Constraint - dny_hair_B_c_l (PP)"!(ComponentPose=@n3, Alpha=1, ConstraintWeights_0=0.5, ConstraintWeights_1:float=0.5) -> (Pose) {
    alphaBoolBlend {blendOption=Linear}
    boneToModify {boneName=dny_hair_B_c_l}
    constraintSetup [{targetBone {boneName=neck_01}}, {targetBone {boneName=head}}]
    constraintWeights [1,1]
   }
   n5 = "Constraint - dny_hair_B_c_r (PP)"!(ComponentPose=@n4, Alpha=1, ConstraintWeights_0=0.5, ConstraintWeights_1:float=0.5) -> (Pose) {
    alphaBoolBlend {blendOption=Linear}
    boneToModify {boneName=dny_hair_B_c_r}
    constraintSetup [{targetBone {boneName=neck_01}}, {targetBone {boneName=head}}]
    constraintWeights [1,1]
   }
   n6 = "Constraint - dny_hair_bk_c (PP)"!(ComponentPose=@n5, Alpha=1, ConstraintWeights_0=0.5, ConstraintWeights_1:float=0.5) -> (Pose) {
    alphaBoolBlend {blendOption=Linear}
    boneToModify {boneName=dny_hair_bk_c}
    constraintSetup [{targetBone {boneName=neck_01}}, {targetBone {boneName=head}}]
    constraintWeights [1,1]
   }
   n7 = "Constraint - dny_hair_A_c_r (PP)"!(ComponentPose=@n6, Alpha=1, ConstraintWeights_0=0.5, ConstraintWeights_1:float=0.5) -> (Pose) {
    alphaBoolBlend {blendOption=Linear}
    boneToModify {boneName=dny_hair_A_c_r}
    constraintSetup [{targetBone {boneName=spine_05}}, {targetBone {boneName=clavicle_r}}]
    constraintWeights [1,1]
   }
   n8 = "Constraint - dny_hair_B_d_l (PP)"!(ComponentPose=@n7, Alpha=1, ConstraintWeights_0=0.5, ConstraintWeights_1:float=0.5) -> (Pose) {
    alphaBoolBlend {blendOption=Linear}
    boneToModify {boneName=dny_hair_B_d_l}
    constraintSetup [{targetBone {boneName=spine_05}}, {targetBone {boneName=clavicle_l}}]
    constraintWeights [1,1]
   }
   n9 = "Constraint - dny_hair_B_d_r (PP)"!(ComponentPose=@n8, Alpha=1, ConstraintWeights_0=0.5, ConstraintWeights_1:float=0.5) -> (Pose) {
    alphaBoolBlend {blendOption=Linear}
    boneToModify {boneName=dny_hair_B_d_r}
    constraintSetup [{targetBone {boneName=spine_05}}, {targetBone {boneName=clavicle_r}}]
    constraintWeights [1,1]
   }
   n10 = "Constraint - dny_hair_bk_d (P)"(ComponentPose=@n9, Alpha=1, ConstraintWeights_0=1) {
    alphaBoolBlend {blendOption=Linear}
    boneToModify {boneName=dny_hair_bk_d}
    constraintSetup [{targetBone {boneName=spine_05}}]
    constraintWeights [1]
   }
   n11 = "Transform (Modify) Bone - Bone: dny_hair_A_c_r"(ComponentPose=@n10, Alpha=1, Translation="0,0.5,0", Rotation="0,-2,0", Scale="1,1,1") {
    rotationMode=BMM_Additive
    rotationSpace=BCS_BoneSpace
    translationMode=BMM_Additive
    translationSpace=BCS_BoneSpace
    alphaBoolBlend {blendOption=Linear}
    boneToModify {boneName=dny_hair_A_c_r}
    rotation {yaw=-2}
    translation {y=0.5}
   }
   n12 = "Transform (Modify) Bone - Bone: dny_hair_A_d_r"(ComponentPose=@n11, Alpha=1, Translation="0,0,0", Rotation="-10,0,0", Scale="1,1,1") {
    rotationMode=BMM_Additive
    rotationSpace=BCS_BoneSpace
    translationMode=BMM_Additive
    translationSpace=BCS_BoneSpace
    alphaBoolBlend {blendOption=Linear}
    boneToModify {boneName=dny_hair_A_d_r}
    rotation {pitch=-10}
   }
   n13 = Component To Local(ComponentPose=@n12)
   n14 = Get bRigidBodyEnabled()
   n15 = RigidBody(ComponentPose=Local To Component(LocalPose=Use cached pose 'AnimCache'())) {
    bClampLinearTranslationLimitToRefPose=true
    bOverrideWorldGravity=true
    overridePhysicsAsset=&/F_MED_Survival_Specialist_diving
    simulationSpace=BaseBoneSpace
    alphaBoolBlend {blendOption=Linear}
    baseBoneRef {boneName=spine_05}
    componentAppliedLinearAccClamp=<500 500 500>
    componentLinearAccScale=<0.2 0.2 0.2>
    componentLinearVelScale=<0.4 0.4 0.4>
    overrideWorldGravity {z=-930}
   }
   n16 = Component To Local(ComponentPose=@n15)
   n17 = RigidBody(ComponentPose=Local To Component(LocalPose=Use cached pose 'AnimCache'())) {
    bClampLinearTranslationLimitToRefPose=true
    bOverrideWorldGravity=true
    overridePhysicsAsset=&F_MED_Survival_Specialist_diving
    simulationSpace=BaseBoneSpace
    alphaBoolBlend {blendOption=Linear}
    baseBoneRef {boneName=pelvis}
    componentAppliedLinearAccClamp=<3000 3000 3000>
    componentLinearAccScale {z=0.6}
    componentLinearVelScale=<0.8 0.8 0.25>
    overrideWorldGravity {z=-930}
   }
   n18 = RigidBody(ComponentPose=Local To Component(LocalPose=Use cached pose 'AnimCache'())) {
    bClampLinearTranslationLimitToRefPose=true
    bOverrideWorldGravity=true
    overridePhysicsAsset=&/F_MED_Survival_Specialist_Autumn_Physics
    simulationSpace=BaseBoneSpace
    alphaBoolBlend {blendOption=Linear}
    baseBoneRef {boneName=spine_05}
    componentAppliedLinearAccClamp=<1000 1000 1000>
    componentLinearAccScale=<0.36 0.36 0.36>
    componentLinearVelScale=<0.3 0.3 0.3>
    overrideWorldGravity {z=-980}
   }
   n19 = "Constraint - dny_hair_B_f_r (PP)"!(ComponentPose=@n18, Alpha=1, ConstraintWeights_0=0.4, ConstraintWeights_1:float=0.6) -> (Pose) {
    alphaBoolBlend {blendOption=Linear}
    boneToModify {boneName=dny_hair_B_f_r}
    constraintSetup [{targetBone {boneName=dny_hair_bk_f}}, {targetBone {boneName=dny_hair_B_f_r}}]
    constraintWeights [1,1]
   }
   n20 = "Constraint - dny_hair_B_f_l (PP)"!(ComponentPose=@n19, Alpha=1, ConstraintWeights_0=0.4, ConstraintWeights_1:float=0.6) -> (Pose) {
    alphaBoolBlend {blendOption=Linear}
    boneToModify {boneName=dny_hair_B_f_l}
    constraintSetup [{targetBone {boneName=dny_hair_bk_f}}, {targetBone {boneName=dny_hair_B_f_l}}]
    constraintWeights [1,1]
   }
   n21 = "Constraint - dny_hair_B_c_r (PP)"(ComponentPose=@n20, Alpha=1, ConstraintWeights_0=0.4, ConstraintWeights_1=0.6) {
    alphaBoolBlend {blendOption=Linear}
    boneToModify {boneName=dny_hair_B_c_r}
    constraintSetup [{targetBone {boneName=dny_hair_bk_c}}, {targetBone {boneName=dny_hair_B_c_r}}]
    constraintWeights [1,1]
   }
   n22 = "Constraint - dny_hair_B_c_l (PP)"(ComponentPose=@n21, Alpha=1, ConstraintWeights_0=0.4, ConstraintWeights_1=0.6) {
    alphaBoolBlend {blendOption=Linear}
    boneToModify {boneName=dny_hair_B_c_l}
    constraintSetup [{targetBone {boneName=dny_hair_bk_c}}, {targetBone {boneName=dny_hair_B_c_l}}]
    constraintWeights [1,1]
   }
   n23 = "Transform (Modify) Bone - Bone: dny_hair_A_a_r"(ComponentPose=@n22, Alpha=1, Translation="0,0,0", Rotation="10,0.000,0", Scale="1,1,1") {
    alpha=0
    rotationMode=BMM_Additive
    rotationSpace=BCS_BoneSpace
    scaleMode=BMM_Additive
    scaleSpace=BCS_BoneSpace
    translationMode=BMM_Additive
    translationSpace=BCS_BoneSpace
    alphaBoolBlend {blendOption=Linear}
    alphaScaleBiasClamp {bInterpResult=true, interpSpeedIncreasing=2, inRange {max=15}}
    boneToModify {boneName=dny_hair_A_a_r}
    rotation {pitch=10}
   }
   n24 = "Constraint - dny_hair_A_c_r (P)"(ComponentPose=@n23, Alpha=1, ConstraintWeights_0=1) {
    alpha=0
    alphaBoolBlend {blendOption=Linear}
    alphaScaleBiasClamp {bInterpResult=true, interpSpeedDecreasing=1, interpSpeedIncreasing=2}
    boneToModify {boneName=dny_hair_A_c_r}
    constraintSetup [{targetBone {boneName=neck_01}}]
    constraintWeights [1]
   }
   n25 = Two Way Blend!(A=Component To Local(ComponentPose=@n24), B=Component To Local(ComponentPose=@n17), bAlphaBoolEnabled:bool=Get bIsSkyDiving()) -> (Pose) {blendNode {alphaInputType=Bool, alphaBoolBlend {blendInTime=0.05, blendOption=Cubic, blendOutTime=0.1}}}
   n26 = Get bIsSwimSprinting() -> (bIsSwimSprinting:bool)
   n27 = "Transform (Modify) Bone - Bone: dny_hair_B_a_base_l"(
    ComponentPose=Local To Component(LocalPose=Two Way Blend(A=@n25, B=@n16, bAlphaBoolEnabled=@n26) {blendNode {alphaInputType=Bool, alphaBoolBlend {blendInTime=0.05, blendOption=Cubic, blendOutTime=0.1}}})
    Alpha=1
    Translation="0,0,1"
    Rotation="0,0,0"
    Scale="1,1,1"
   ) {
    rotationMode=BMM_Additive
    rotationSpace=BCS_BoneSpace
    scaleMode=BMM_Additive
    scaleSpace=BCS_BoneSpace
    translationMode=BMM_Additive
    translationSpace=BCS_BoneSpace
    alphaBoolBlend {blendOption=Linear}
    alphaScaleBiasClamp {inRange {max=15}}
    boneToModify {boneName=dny_hair_B_a_base_l}
    translation {z=1}
   }
   n28 = "Transform (Modify) Bone - Bone: dny_hair_B_a_r"(ComponentPose=@n27, Alpha=Get headAimUp() -> (headAimUp:double), Translation="-2,0,0", Rotation="0,0,0", Scale="1,1,1") {
    translationMode=BMM_Additive
    translationSpace=BCS_BoneSpace
    alphaBoolBlend {blendOption=Linear}
    boneToModify {boneName=dny_hair_B_a_r}
    translation {x=-2}
   }
   n29 = Get headAimUp()
   n30 = "Transform (Modify) Bone - Bone: dny_hair_B_b_l"(ComponentPose=@n28, Alpha=@n29, Translation="0,3,0", Rotation="-10,0,0", Scale="1,1,1") {
    translationMode=BMM_Additive
    translationSpace=BCS_BoneSpace
    alphaBoolBlend {blendOption=Linear}
    boneToModify {boneName=dny_hair_B_b_l}
    rotation {pitch=-10}
    translation {y=3}
   }
   n31 = "Transform (Modify) Bone - Bone: dny_hair_B_d_l"(ComponentPose=@n30, Alpha=@n29, Translation="0,5,0", Rotation="0,0,0", Scale="1,1,1") {
    translationMode=BMM_Additive
    translationSpace=BCS_BoneSpace
    alphaBoolBlend {blendOption=Linear}
    boneToModify {boneName=dny_hair_B_d_l}
    translation {y=5}
   }
   n32 = "Transform (Modify) Bone - Bone: dny_hair_A_c_r"(ComponentPose=@n31, Alpha=1, Translation="-2.072,-2,-2.800", Rotation="-9.847,0,0", Scale="1,1,1") {
    rotationMode=BMM_Additive
    translationMode=BMM_Additive
    alphaBoolBlend {blendOption=Linear}
    boneToModify {boneName=dny_hair_A_c_r}
    rotation {pitch=-9.847}
    translation=<-2.072 -2 -2.8>
   }
   Output Pose(Result=Blend Poses by bool(
     BlendPose_0=Blend Poses by bool(BlendPose_0=@n2, BlendPose_1=@n13, bActiveValue=@n14) {blendType=Linear}
     BlendPose_1=Component To Local(ComponentPose=@n32)
     bActiveValue=Get bIsInFrontEnd()
    ) {blendType=Linear}) {layerGroup=None}
   n33 = Get headUp() -> (headUp:double)
   n34 = "Transform (Modify) Bone - Bone: dny_hair_bk_a_base"(
    ComponentPose=Local To Component(LocalPose=Copy Pose From Mesh() {bCopyCurves=true, bUseAttachedParent=true})
    Alpha=@n33
    Translation="0,-3,0"
    Rotation="0,-25,0"
    Scale="1,1,1"
   ) {
    rotationMode=BMM_Additive
    rotationSpace=BCS_BoneSpace
    translationMode=BMM_Additive
    translationSpace=BCS_BoneSpace
    alphaBoolBlend {blendOption=Linear}
    alphaScaleBiasClamp {bMapRange=true, inRange {max=0, min=-50}, outRange {max=0, min=1}}
    boneToModify {boneName=dny_hair_bk_a_base}
    rotation {yaw=-25}
    translation {y=-3}
   }
   n35 = "Transform (Modify) Bone - Bone: dny_hair_B_a_base_l"(ComponentPose=@n34, Alpha=@n33, Translation="0,-4,2", Rotation="0,-25,0", Scale="0.7,1,1") {
    rotationMode=BMM_Additive
    rotationSpace=BCS_BoneSpace
    scaleMode=BMM_Additive
    scaleSpace=BCS_BoneSpace
    translationMode=BMM_Additive
    translationSpace=BCS_BoneSpace
    alphaBoolBlend {blendOption=Linear}
    alphaScaleBiasClamp {bMapRange=true, inRange {max=0, min=-50}, outRange {max=0, min=1}}
    boneToModify {boneName=dny_hair_B_a_base_l}
    rotation {yaw=-25}
    scale {x=0.7}
    translation {y=-4, z=2}
   }
   n36 = Get headUp()
   n37 = "Transform (Modify) Bone - Bone: dny_hair_B_a_base_r"(ComponentPose=@n35, Alpha=@n36, Translation="0,-1,-2", Rotation="0,-40,0", Scale="0.7,1,1") {
    rotationMode=BMM_Additive
    rotationSpace=BCS_BoneSpace
    scaleMode=BMM_Additive
    scaleSpace=BCS_BoneSpace
    translationMode=BMM_Additive
    translationSpace=BCS_BoneSpace
    alphaBoolBlend {blendOption=Linear}
    alphaScaleBiasClamp {bMapRange=true, inRange {max=0, min=-50}, outRange {max=0, min=1}}
    boneToModify {boneName=dny_hair_B_a_base_r}
    rotation {yaw=-40}
    scale {x=0.7}
    translation {y=-1, z=-2}
   }
   n38 = "Transform (Modify) Bone - Bone: dny_hair_A_a_base_r"(ComponentPose=@n37, Alpha=@n36, Translation="0,0,0", Rotation="0,-25,0", Scale="1,1,1") {
    rotationMode=BMM_Additive
    rotationSpace=BCS_BoneSpace
    scaleMode=BMM_Additive
    scaleSpace=BCS_BoneSpace
    translationMode=BMM_Additive
    translationSpace=BCS_BoneSpace
    alphaBoolBlend {blendOption=Linear}
    alphaScaleBiasClamp {bMapRange=true, inRange {max=0, min=-50}, outRange {max=0, min=1}}
    boneToModify {boneName=dny_hair_A_a_base_r}
    rotation {yaw=-25}
   }
   n39 = Component To Local(ComponentPose=@n38)
   n40 = Get headUp()
   n41 = "Transform (Modify) Bone - Bone: dny_hair_bk_a_base"(
    ComponentPose=Local To Component(LocalPose=Copy Pose From Mesh() {bCopyCurves=true, bUseAttachedParent=true})
    Alpha=@n40
    Translation="0,-3,0"
    Rotation="0,-25,0"
    Scale="1,1,1"
   ) {
    rotationMode=BMM_Additive
    rotationSpace=BCS_BoneSpace
    translationMode=BMM_Additive
    translationSpace=BCS_BoneSpace
    alphaBoolBlend {blendOption=Linear}
    alphaScaleBiasClamp {bMapRange=true, inRange {max=0, min=-50}, outRange {max=0, min=1}}
    boneToModify {boneName=dny_hair_bk_a_base}
    rotation {yaw=-25}
    translation {y=-3}
   }
   n42 = "Transform (Modify) Bone - Bone: dny_hair_B_a_base_l"(ComponentPose=@n41, Alpha=@n40, Translation="0,-4,2", Rotation="0,-35,0", Scale="0.7,1,1") {
    rotationMode=BMM_Additive
    rotationSpace=BCS_BoneSpace
    scaleMode=BMM_Additive
    scaleSpace=BCS_BoneSpace
    translationMode=BMM_Additive
    translationSpace=BCS_BoneSpace
    alphaBoolBlend {blendOption=Linear}
    alphaScaleBiasClamp {bMapRange=true, inRange {max=0, min=-50}, outRange {max=0, min=1}}
    boneToModify {boneName=dny_hair_B_a_base_l}
    rotation {yaw=-35}
    scale {x=0.7}
    translation {y=-4, z=2}
   }
   n43 = Get headUp()
   n44 = "Transform (Modify) Bone - Bone: dny_hair_B_a_base_r"(ComponentPose=@n42, Alpha=@n43, Translation="0,-1,-2", Rotation="0,-55,0", Scale="0.7,1,1") {
    rotationMode=BMM_Additive
    rotationSpace=BCS_BoneSpace
    scaleMode=BMM_Additive
    scaleSpace=BCS_BoneSpace
    translationMode=BMM_Additive
    translationSpace=BCS_BoneSpace
    alphaBoolBlend {blendOption=Linear}
    alphaScaleBiasClamp {bMapRange=true, inRange {max=0, min=-50}, outRange {max=0, min=1}}
    boneToModify {boneName=dny_hair_B_a_base_r}
    rotation {yaw=-55}
    scale {x=0.7}
    translation {y=-1, z=-2}
   }
   n45 = "Transform (Modify) Bone - Bone: dny_hair_A_a_base_r"(ComponentPose=@n44, Alpha=@n43, Translation="0,0,0", Rotation="0,-25,0", Scale="1,1,1") {
    rotationMode=BMM_Additive
    rotationSpace=BCS_BoneSpace
    scaleMode=BMM_Additive
    scaleSpace=BCS_BoneSpace
    translationMode=BMM_Additive
    translationSpace=BCS_BoneSpace
    alphaBoolBlend {blendOption=Linear}
    alphaScaleBiasClamp {bMapRange=true, inRange {max=0, min=-50}, outRange {max=0, min=1}}
    boneToModify {boneName=dny_hair_A_a_base_r}
    rotation {yaw=-25}
   }
   n46 = Component To Local(ComponentPose=@n45)
   n47 = Get bIsSkyDiving()
   n48 = Get headUp()
   n49 = "Transform (Modify) Bone - Bone: dny_hair_bk_a_base"(
    ComponentPose=Local To Component(LocalPose=Two Way Blend(A=@n39, B=@n46, bAlphaBoolEnabled=@n47) {blendNode {alphaInputType=Bool, alphaBoolBlend {blendInTime=0.05, blendOption=Cubic, blendOutTime=0.1}}})
    Alpha=@n48
    Translation="0,0,0"
    Rotation="0,10,0"
    Scale="1,1,1"
   ) {
    rotationMode=BMM_Additive
    rotationSpace=BCS_BoneSpace
    translationMode=BMM_Additive
    translationSpace=BCS_BoneSpace
    alphaBoolBlend {blendOption=Linear}
    alphaScaleBiasClamp {bMapRange=true, inRange {max=15}}
    boneToModify {boneName=dny_hair_bk_a_base}
    rotation {yaw=10}
   }
   n50 = "Transform (Modify) Bone - Bone: dny_hair_B_a_base_l"(ComponentPose=@n49, Alpha=@n48, Translation="0,0,0", Rotation="0,15,0", Scale="1,1,1") {
    rotationMode=BMM_Additive
    rotationSpace=BCS_BoneSpace
    scaleMode=BMM_Additive
    scaleSpace=BCS_BoneSpace
    translationMode=BMM_Additive
    translationSpace=BCS_BoneSpace
    alphaBoolBlend {blendOption=Linear}
    alphaScaleBiasClamp {bMapRange=true, inRange {max=15}}
    boneToModify {boneName=dny_hair_B_a_base_l}
    rotation {yaw=15}
   }
   n51 = Get headUp()
   n52 = "Transform (Modify) Bone - Bone: dny_hair_B_a_base_r"(ComponentPose=@n50, Alpha=@n51, Translation="0,0,0", Rotation="0,15,0", Scale="1,1,1") {
    rotationMode=BMM_Additive
    rotationSpace=BCS_BoneSpace
    scaleMode=BMM_Additive
    scaleSpace=BCS_BoneSpace
    translationMode=BMM_Additive
    translationSpace=BCS_BoneSpace
    alphaBoolBlend {blendOption=Linear}
    alphaScaleBiasClamp {bMapRange=true, inRange {max=15}}
    boneToModify {boneName=dny_hair_B_a_base_r}
    rotation {yaw=15}
   }
   n53 = "Transform (Modify) Bone - Bone: dny_hair_A_a_base_r"(ComponentPose=@n52, Alpha=@n51, Translation="0,0,0", Rotation="0,20,0", Scale="1,1,1") {
    rotationMode=BMM_Additive
    rotationSpace=BCS_BoneSpace
    scaleMode=BMM_Additive
    scaleSpace=BCS_BoneSpace
    translationMode=BMM_Additive
    translationSpace=BCS_BoneSpace
    alphaBoolBlend {blendOption=Linear}
    alphaScaleBiasClamp {bMapRange=true, inRange {max=15}}
    boneToModify {boneName=dny_hair_A_a_base_r}
    rotation {yaw=20}
   }
   n54 = "Transform (Modify) Bone - Bone: dny_hair_B_b_l"(ComponentPose=@n53, Alpha=1, Translation="0,0,0", Rotation="0,-5,0", Scale="1,1,1") {
    rotationMode=BMM_Additive
    rotationSpace=BCS_BoneSpace
    scaleMode=BMM_Additive
    scaleSpace=BCS_BoneSpace
    translationMode=BMM_Additive
    translationSpace=BCS_BoneSpace
    alphaBoolBlend {blendOption=Linear}
    alphaScaleBiasClamp {bMapRange=true, inRange {max=15}}
    boneToModify {boneName=dny_hair_B_b_l}
    rotation {yaw=-5}
   }
   n55 = "Transform (Modify) Bone - Bone: dny_hair_B_b_r"(ComponentPose=@n54, Alpha=1, Translation="0,0,0", Rotation="0,-5,0", Scale="1,1,1") {
    rotationMode=BMM_Additive
    rotationSpace=BCS_BoneSpace
    scaleMode=BMM_Additive
    scaleSpace=BCS_BoneSpace
    translationMode=BMM_Additive
    translationSpace=BCS_BoneSpace
    alphaBoolBlend {blendOption=Linear}
    alphaScaleBiasClamp {bMapRange=true, inRange {max=15}}
    boneToModify {boneName=dny_hair_B_b_r}
    rotation {yaw=-5}
   }
   n56 = "Transform (Modify) Bone - Bone: dny_hair_bk_b"(ComponentPose=@n55, Alpha=1, Translation="0,0,0", Rotation="0,-5,0", Scale="1,1,1") {
    rotationMode=BMM_Additive
    rotationSpace=BCS_BoneSpace
    scaleMode=BMM_Additive
    scaleSpace=BCS_BoneSpace
    translationMode=BMM_Additive
    translationSpace=BCS_BoneSpace
    alphaBoolBlend {blendOption=Linear}
    alphaScaleBiasClamp {bMapRange=true, inRange {max=15}}
    boneToModify {boneName=dny_hair_bk_b}
    rotation {yaw=-5}
   }
   Save cached pose 'AnimCache'(Pose=Component To Local(ComponentPose=@n56)) {cacheName=AnimCache}
  ]}
 EventGraph {nodes [
   Event BlueprintUpdateAnimation()
   > Sequence()
   n1 = GetDeltaTransformFromRefPose(self=GetOwnerSkeletalMeshComponent(), BoneName=head, BaseName=spine_05)
   > Set headUp(headUp:double=BreakRotator(InRot=BreakTransform(InTransform=@n1).Rotation).Yaw) -> (Output_Get:double)
   > n2 = Branch(Condition=Get bIsTargeting())
   n3 = BreakTransform(InTransform=GetDeltaTransformFromRefPose(self=GetOwnerSkeletalMeshComponent(), BoneName=spine_05, BaseName=neck_01))
   n4 = MapRangeClamped(Value="Absolute (Float)"(A=BreakRotator(InRot=@n3.Rotation).Yaw), InRangeA=0, InRangeB=10)
   > Set headAimUp(headAimUp:double=@n4) -> (Output_Get:double)
   Set headAimUp(execute=@n2.else, headAimUp=float / float(A=@n4, B=4) {is=title K2Node_CallFunction})
   > n5 = Branch(Condition=IsCosmeticObjectDisplayedInFrontendContext!(CosmeticObject:Object Reference) -> (ReturnValue:bool))
   > Set FrontendAlpha(FrontendAlpha:double=0.5) -> (Output_Get:double)
   Set FrontendAlpha(execute=@n5.else, FrontendAlpha=1)
   Event BlueprintInitializeAnimation()
   > Set Body_Mesh(Body_Mesh=GetOwnerSkeletalMeshComponent())
   > n6 = Cast To PlayerPawn_Athena(Object=TryGetPawnOwner())
   > n7 = Set Pawn Owner(Pawn Owner:Fort Player Pawn Athena Object Reference=@n6) -> (Output_Get:Fort Player Pawn Athena Object Reference)
   > Set Head_Mesh(Head_Mesh:Skeletal Mesh Component Object Reference=GetSkeletalMeshForPartType(self=@n7, PartType=Head)) -> (Output_Get:Skeletal Mesh Component Object Reference)
  ]}
]
variables [
 Anim_Inst {type=AnimInstance, default=null}
 Body_Mesh {type=SkeletalMeshComponent}
 EmoteAlpha_b {type=number}
 FrontendAlpha {type=number}
 Head_Mesh {type=SkeletalMeshComponent}
 headAimUp {type=number}
 headUp {type=number}
 Pawn Owner {type=FortPlayerPawnAthena}
]
